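; Gameplay ability for a BallerCoaster emote-duration quest that reports progress once the player has qualified long enough.
; On event activation it runs a check, then repeats it on a looping timer.
; Each check adds the interval to a never-reset total while the owning pawn has the BallerCoaster tag.
; Past the required time, it sends a quest stat event and clears the timer, which ending the ability also clears.
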(event Ability|EventActivateAbilityFromEvent
  (params (EventData :type Gameplay%20Event%20Data%20Structure%20%28by%20ref%29))
  (|CallParentFunction EventData)
  (Development|PrintString "Start Checking For Sideways Emote" false false "(R=0.000000,G=0.660000,B=1.000000,A=1.000000)" 2e+01)
  (CallFunction|CheckifEmoteRequirementsMet)
  (bind _returnvalue (Utilities|Time|SetTimerbyEvent (delegate Custom|CheckIfEmoteRequirementsMet) (Variables|Default|GetCheckInterval) true))
  (Variables|Default|SetTimer_CheckForEmoteRequirement _returnvalue))

(event Custom|CheckIfEmoteRequirementsMet
  (Development|PrintString "Checking For Sideways Emote" false false)
  (bind _returnvalue (MissionHelpers|HasTags (GameplayTags|GetOwnedGameplayTags (Variables|GAVars|Data|GetOwningPlayerPawn) :!class /Script/GameplayTags.GameplayTagAssetInterface) (Variables|Default|GetBallerCoasterTag) false false))
  (if _returnvalue
    (bind _output_get (Variables|Default|SetTimeEmoting (+ (Variables|Default|GetTimeEmoting) (Variables|Default|GetCheckInterval))))
    (if (> _output_get (Variables|Default|GetTimeRequiredToEmote))
      (Development|PrintString "VladDebug Emoted and Broadcasted ")
      (Quest|SendComplexCustomStatEvent (Variables|GAVars|Data|GetOwningPlayerQuestManager) 0 (Variables|Default|GetQuestTag) "(GameplayTags=)" 1)
      (Utilities|Time|ClearandInvalidateTimerbyHandle (Variables|Default|GetTimer_CheckForEmoteRequirement)))))

(event Ability|EventOnEndAbility
  (params (bWasCancelled :type Boolean))
  (bind _timer_checkforemoterequirement (Variables|Default|GetTimer_CheckForEmoteRequirement))
  (Development|PrintString "Stop Checking For Sideways Emote" false false "(R=0.000000,G=0.660000,B=1.000000,A=1.000000)" 2e+01)
  (if (Utilities|Time|IsValidTimerHandle _timer_checkforemoterequirement)
    (Utilities|Time|ClearandInvalidateTimerbyHandle _timer_checkforemoterequirement)))
